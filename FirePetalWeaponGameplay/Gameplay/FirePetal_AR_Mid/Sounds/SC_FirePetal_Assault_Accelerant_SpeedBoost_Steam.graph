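# Sound cue for the steam effect of the FirePetal mid assault rifle's Accelerant speed boost.
# A mixer plays a one-shot steam start wave alongside a looping steam wave held back by a short fixed delay, so the start leads into the sustained loop.
# The cue uses the Fort weapon foley sound class and attenuation settings with a limited max distance.
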SoundCue /FirePetalWeaponGameplay/Gameplay/FirePetal_AR_Mid/Sounds/SC_FirePetal_Assault_Accelerant_SpeedBoost_Steam
root=Mixer_0
nodes [
 {id=Delay_0, children [WavePlayer_0], properties {delayMax=0.1, delayMin=0.1}}
 {id=Mixer_0, children [WavePlayer_1,Delay_0], properties {inputVolume [1,1]}}
 {
  id=WavePlayer_0
  properties {
   bLooping=true
   soundWaveAssetPtr=/FirePetalWeaponGameplay/Gameplay/FirePetal_AR_Mid/Sounds/wavs/Assault_FirePetal_Mid_Accelerant_SpeedBoost_SteamLoop
  }
 }
 {
  id=WavePlayer_1
  properties {soundWaveAssetPtr=/FirePetalWeaponGameplay/Gameplay/FirePetal_AR_Mid/Sounds/wavs/Assault_FirePetal_Mid_Accelerant_SpeedBoost_SteamStart}
 }
]
properties {
 attenuationSettings=&AS_Fort_WeaponFoley
 maxDistance=2250
 soundClassObject=&Fort_WeaponFoley
}
